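; Server-side spawner that drops Stamina capsule pods during storm circles, gated by a random roll and calendar event.
; At startup it picks drop counts by playlist tag and caches landing points from an EQS query.
; Each StormHolding phase rolls a chance, then spawns staggered, angled drops onto shuffled safe-zone points until the match total is reached.
; It also holds a cheat spawn and an uncalled-here SpawnV2 targeting the most crowded player.
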(event EventBeginPlay
  ;[1 1 1] BeginPlay
  (bind (_gamestate _playlist _playlistcontexttags) (Class|FortAsyncActionCurrentPlaylistReady|PlaylistDataReadyAsync)
    (:OnReady
      (if (Networking|IsServer)
        (if (>= (ScalableFloat|GetValueatLevel (Variables|Default|GetRow_ChanceTospawnAtAll)) (Math|Random|RandomFloat))
          ;;[1 1 1] Check if calendar event is active
          (bind _biseventactive (CalendarEvents|GetCalendarEventInformation (Variables|Default|GetCalendarEventName)))
          (if (or _biseventactive (ScalableFloat|GetValueasBool (Variables|Default|GetRow_IgnoreCalendarEvent)))
            (CallFunction|SetupDropParams _playlistcontexttags)
            (CallFunction|RunDropPointQuery)
            (bind _returnvalue (GameplayEventRouter|GetEventRouter "/Script/FortniteGame.FortGameState" self))
            (bind (_asynctaskproxy _payload) (GameplayEventRouter|StartListeningToStatefulEvent _returnvalue "/Script/FortniteGame.FortGamePhaseStepUpdatedEvent")
              (:then
                (Variables|Default|SetGamePhaseStepUpdatedAsyncTask _asynctaskproxy))
              (:OnEventReceived
                _)
              (:OnSavedEventStateExists
                _))
            (CallFunction|OnGameStepChanged (Utilities|Struct|BreakFortGamePhaseStepUpdatedEvent _payload))
            (else
              (Actor|DestroyActor)))
          (else
            (Actor|DestroyActor)))))))

(event Custom|OnGameStepChanged
  (params (GamePhaseStep :type EAthenaGamePhaseStep%20Enum))
  ;[1 1 1] On Game Step Changed - on new storm circle, check if drop pods should spawn based on random chance
  (if (Utilities|Enum|Equal(Enum) GamePhaseStep "StormHolding")
    (Math|Integer|IncrementInt (Variables|Default|GetCircleCount)
      (:%20%20
        (CallFunction|SetupPerCircleDropParams)
        (if (Math|Float|InRange(Float) (Math|Random|RandomFloatinRange 0.0 1.0) 0.0 (ScalableFloat|GetValueatLevel (Variables|Default|GetRow_DropSpawnChance)))
          (Utilities|FlowControl|Delay 3.0)
          (CallFunction|SpawnPods))))))

(event Custom|SpawnPods
  ;[1 1 1] EQS Query callback -> spawn drop pods above locations from query
  (if (>= (Variables|Default|GetTotalSpawnCount) (Variables|Default|GetFinalTotalNumDrop))
    (if (Variables|Default|GetUseCheatLandingLocation)
      _))
  (if (Variables|Default|GetUseCheatLandingLocation)
    (Variables|Default|SetLoopLandingLoc (Variables|Default|GetCheatLandingLocation))
    (else
      (CallFunction|PrepareLocations)
      (bind _result (ArrayMacros|SafeArrayGet (Variables|Default|GetSafeLocations) (Variables|Default|GetPerCircleSpawnCount))
        (:%20%20
          (Variables|Default|SetLoopLandingLoc _result)))))
  (CallFunction|CalculateSpawnRot&Loc (Variables|Default|GetLoopLandingLoc))
  (Variables|Default|SetUseCheatLandingLocation false)
  (Game|SpawnActorfromClass "/StaminaGameplay/Gameplay/Environmental/Capsule/BGA_Stamina_Capsule_Drop.BGA_Stamina_Capsule_Drop_C" (Math|Transform|MakeTransform (Variables|Default|GetspawnLocation) (Math|Rotator|FindLookatRotation (Variables|Default|GetspawnLocation) (Variables|Default|GetLoopLandingLoc))) "AlwaysSpawn" "OverrideRootScale" 0 (Variables|Default|GetFallDirection) (Variables|Default|GetLoopLandingLoc) 0 0)
  (Math|Integer|IncrementInt (Variables|Default|GetSpawnCountThisCircle)
    (:%20%20
      (Math|Integer|IncrementInt (Variables|Default|GetTotalSpawnCount)
        (:%20%20
          (if (>= (Variables|Default|GetTotalSpawnCount) (Variables|Default|GetFinalTotalNumDrop))
            (CallFunction|Finished)
            (elif (< (Variables|Default|GetSpawnCountThisCircle) (Variables|Default|GetFinalNumSpawnPerCircle))
              (Utilities|FlowControl|Delay 3.0)
              (CallFunction|SpawnPods)
              (else
                (Variables|Default|SetSpawnCountThisCircle 0)))))))))

(event Custom|PlayDropIndicationSound
  (params (Pos :type Vector)))

(event Custom|Finished
  ;[1 1 1] unbind
  (bind _gamephasestepupdatedasynctask (Variables|Default|GetGamePhaseStepUpdatedAsyncTask)
    (:then
      (GameplayEventRouter|StopListeningtoStatefulEvent _gamephasestepupdatedasynctask))
    (:else)))

(event Custom|OnDropPointQueryFinished
  (params (QueryInstance :type EQS%20Query%20Instance%20Object%20Reference) (QueryStatus :type EEnvQueryStatus%20Enum :default "Processing"))
  ;[1 1 1] EQS Query callback
  (bind (_resultlocations _returnvalue) (AI|EQS|GetQueryResultsasLocations QueryInstance))
  (if _returnvalue
    (Variables|Default|SetResult%20Locations _resultlocations)))

(event Custom|PrepareLocations
  ;[1 1 1] Prepare Safe Locations
  (Utilities|Array|Clear (Variables|Default|GetSafeLocations))
  (for _array_element (Variables|Default|GetResult%20Locations)
    (bind _returnvalue (SafeZone|IsLocationinSafeZone _array_element))
    (if _returnvalue
      (Utilities|Array|Add (Variables|Default|GetSafeLocations) _array_element)))
  (Utilities|Array|Shuffle (Variables|Default|GetSafeLocations)))

(event Custom|SpawnV2
  (bind _targetplayer (CallFunction|FindTargetPlayer))
  (Utilities|IsValid _targetplayer
    (:Is%20Valid
      (Utilities|IsValid (Pawn|GetControlledPawn _targetplayer)
        (:Is%20Valid
          (Game|SpawnActorfromClass "/StaminaGameplay/Gameplay/Environmental/Capsule/BGA_Stamina_Capsule_Drop.BGA_Stamina_Capsule_Drop_C" (Math|Conversions|ToTransform(Vector) (+ (* (Math|Vector|Normalize2D(Vector) (Math|Vector|GetForwardVector (Pawn|GetControlRotation _targetplayer :!class /Script/Engine.Controller))) 5e+03) (+ (Transformation|GetActorLocation (Pawn|GetControlledPawn _targetplayer)) "0,0,10000.000000"))) "AlwaysSpawn" "OverrideRootScale" 0 "0.000000,0.000000,0.000000" "0.000000,0.000000,0.000000" 0 0))))))

(fn CalculateSpawnRot&Loc
  (params (landingLocation :type Vector))
  (LocalVariables|SetLandingLoc landingLocation)
  (Variables|Default|SetFallDirection (Math|Vector|CreateVectorfromYawPitch (Math|Random|RandomFloatinRange -1.8e+02 1.8e+02) (* (Math|Random|RandomFloatinRange (ScalableFloat|GetValueatLevel (Variables|Default|GetRow_FallAngleMin)) (ScalableFloat|GetValueatLevel (Variables|Default|GetRow_FallAngleMax))) -1.0))) ; Specified Fall Angle & height with random yaw
  (Variables|Default|SetspawnLocation (+ (+ (LocalVariables|GetLandingLoc) "0,0,50000.000000") (Utilities|Operators|Multiply (Math|Vector|Normalize2D(Vector) (Variables|Default|GetFallDirection)) (ScalableFloat|GetValueatLevel (Variables|Default|GetRow_FallDistance)) -1.0)))
  (if (> (Math|Vector|VectorLengthXY (Variables|Default|GetspawnLocation)) (Variables|Default|GetMapRadius))
    (LocalVariables|SetSpawnLocZ (.z (Variables|Default|GetspawnLocation)))
    (Variables|Default|SetspawnLocation (Math|Vector|MakeVector (.x (* (Math|Vector|Normalize2D(Vector) (Variables|Default|GetspawnLocation)) (Variables|Default|GetMapRadius))) (.y (* (Math|Vector|Normalize2D(Vector) (Variables|Default|GetspawnLocation)) (Variables|Default|GetMapRadius))) (LocalVariables|GetSpawnLocZ)))
    (else
      _))
  (Variables|Default|SetFallDirection (Math|Vector|Normalize (- (LocalVariables|GetLandingLoc) (Variables|Default|GetspawnLocation))))
  (return))

(fn FindTargetPlayer
  ;[1 1 1] This search is N*N algorithm currently, might not need to be optimized since it happens very infrequently
  (LocalVariables|SetCurrentBestPlayerCount 0)
  (LocalVariables|SetCurrentBestController 0)
  (LocalVariables|SetAllPlayerControllers (Fortnite|GetAllFortPlayerControllers true))
  (for _array_element (LocalVariables|GetAllPlayerControllers)
    (bind _outnumplayersinrange (CallFunction|GetNumPlayersNearPlayer _array_element (Variables|Default|GetPlayerSearchRadius)))
    (if (> _outnumplayersinrange (LocalVariables|GetCurrentBestPlayerCount))
      (LocalVariables|SetCurrentBestController _array_element)
      (LocalVariables|SetCurrentBestPlayerCount _outnumplayersinrange)))
  (return (LocalVariables|GetCurrentBestController)))

(fn GetNumPlayersNearPlayer
  (params (InController :type Fort%20Player%20Controller%20Object%20Reference) (Range :type Float%20%28double-precision%29))
  (bind _controller (LocalVariables|GetController))
  (bind _returnvalue (Pawn|GetControlledPawn _controller))
  (LocalVariables|SetInRangeSquared (* Range Range))
  (LocalVariables|SetController InController)
  (Utilities|IsValid _returnvalue
    (:Is%20Valid
      (LocalVariables|SetInPlayerLoc (Transformation|GetActorLocation _returnvalue))
      (for _array_element (Fortnite|GetAllFortPlayerControllers true)
        (Utilities|IsValid (Pawn|GetControlledPawn _array_element)
          (:Is%20Valid
            (if (<= (Math|Vector|Distance2DSquared(Vector) (LocalVariables|GetInPlayerLoc) (Transformation|GetActorLocation (Pawn|GetControlledPawn _array_element))) (LocalVariables|GetInRangeSquared))
              (Math|Integer|IncrementInt (LocalVariables|GetPlayerCount))))))
      (return (LocalVariables|GetPlayerCount)))))

(fn GetRandomDrops
  (params (Min :type Scalable%20Float%20Structure) (Max :type Scalable%20Float%20Structure))
  (return (Math|Random|RandomIntegerinRange (Math|Float|Truncate (ScalableFloat|GetValueatLevel Min)) (Math|Float|Truncate (ScalableFloat|GetValueatLevel Max)))))

(fn GetRandomPointFromQueryResults
  (params (Points :type Array%20of%20Vectors))
  (bind _points (Variables|GetPoints))
  (bind _returnvalue (Utilities|Array|Length _points))
  (bind _returnvalue_2 (Math|Random|RandomIntegerinRange 0 _returnvalue))
  (LocalVariables|SetFoundGoodIndex false)
  (while (and (not (LocalVariables|GetFoundGoodIndex)) (< (LocalVariables|GetNumLoops) (LocalVariables|GetMaxLoops)))
    (Math|Integer|IncrementInt (LocalVariables|GetNumLoops)
      (:%20%20
        (if (not (Utilities|Array|ContainsItem (Variables|Default|GetLocationIndicesUsed) _returnvalue_2))
          (LocalVariables|SetFoundGoodIndex true)
          (bind _output_get (LocalVariables|SetValidIndex _returnvalue_2))
          (Utilities|Array|Add (Variables|Default|GetLocationIndicesUsed) _output_get)))))
  (return (Utilities|Array|Get(acopy) Points (LocalVariables|GetValidIndex))))

(fn RunDropPointQuery
  (if (not (Variables|Default|GetDidEQSQuery))
    (bind _returnvalue (AI|EQS|RunEQSQuery (Variables|Default|GetQuery%20Template) self "AllMatching"))
    (AI|EQS|BindEventtoOnQueryFinished _returnvalue (EventDispatchers|CreateEvent :!event OnDropPointQueryFinished))
    (Variables|Default|SetDidEQSQuery true)))

(fn SetupDropParams
  (params (InTags :type Gameplay%20Tag%20Container%20Structure))
  (if (GameplayTags|HasTag InTags (Variables|Default|GetTeamRumbleTag))
    (Variables|Default|SetFinalNumSpawnPerCircle (Math|Float|Truncate (ScalableFloat|GetValueatLevel (Variables|Default|GetRow_DropsPerCircle_TeamRumble))))
    (Variables|Default|SetFinalTotalNumDrop (Math|Float|Truncate (ScalableFloat|GetValueatLevel (Variables|Default|GetRow_TotalNumToDrop_TeamRumble))))
    (elif (GameplayTags|HasTag InTags (Variables|Default|GetBigTeamTag))
      (Variables|Default|SetFinalNumSpawnPerCircle (Math|Float|Truncate (ScalableFloat|GetValueatLevel (Variables|Default|GetRow_DropsPerCircle_BigTeam))))
      (Variables|Default|SetFinalTotalNumDrop (Math|Float|Truncate (ScalableFloat|GetValueatLevel (Variables|Default|GetRow_TotalNumToDrop_BigTeam))))
      (else
        (Variables|Default|SetFinalNumSpawnPerCircle (Math|Float|Truncate (ScalableFloat|GetValueatLevel (Variables|Default|GetRow_NumToSpawnPerCircle))))
        (Variables|Default|SetFinalTotalNumDrop (Math|Float|Truncate (ScalableFloat|GetValueatLevel (Variables|Default|GetRow_TotalNumToDrop))))
        (Variables|Default|SetInBR true)))))

(fn SetupPerCircleDropParams
  (if (Variables|Default|GetInBR)
    (if (== (Variables|Default|GetCircleCount) 1)
      (bind _out (CallFunction|GetRandomDrops (Variables|Default|GetRow_MinCircle1) (Variables|Default|GetRow_MaxCircle1)))
      (Variables|Default|SetFinalNumSpawnPerCircle _out)
      (elif (== (Variables|Default|GetCircleCount) 2)
        (bind _out_2 (CallFunction|GetRandomDrops (Variables|Default|GetRow_MinCircle2) (Variables|Default|GetRow_MaxCircle2)))
        (Variables|Default|SetFinalNumSpawnPerCircle _out_2)
        (elif (== (Variables|Default|GetCircleCount) 3)
          (bind _out_3 (CallFunction|GetRandomDrops (Variables|Default|GetRow_MinCircle3) (Variables|Default|GetRow_MaxCircle3)))
          (Variables|Default|SetFinalNumSpawnPerCircle _out_3)
          (else
            (bind _out_4 (CallFunction|GetRandomDrops (Variables|Default|GetRow_MinCircle4) (Variables|Default|GetRow_MaxCircle4)))
            (Variables|Default|SetFinalNumSpawnPerCircle _out_4)))))))

(fn SpawnFromCheat
  (bind _asfort_player_controller (Utilities|Casting|CastToFortPlayerController (Game|GetPlayerController 0)))
  (bind _returnvalue (Fortnite|GetPlayerAimPoint _asfort_player_controller 1e+04 (Utilities|Array|MakeArray 0)))
  (Variables|Default|SetCheatLandingLocation _returnvalue)
  (Variables|Default|SetUseCheatLandingLocation true)
  (CallFunction|SpawnPods))

(fn ConstructionScript)
