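# EQS query for the Stamina Hot Potato drop pod, apparently selecting candidate locations for it across the map.
# A very wide, coarsely spaced SimpleGrid projects points vertically onto geometry with sphere traces on a custom channel rather than navmesh.
# Points are filtered to the Athena safe zone, checked by two differently sized overlap tests on separate custom channels, then run through a random test set to filter.
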EnvQuery /StaminaGameplay/Gameplay/Environmental/EQS_Stamina_HotPotato_DropPod
options [{
  generator {
   name=EnvQueryGenerator_SimpleGrid_0
   class=EnvQueryGenerator_SimpleGrid
   properties {
    gridSize=200000
    spaceBetween=3000
    projectionData {
     extentX=0
     navigationFilter=null
     postProjectionVerticalOffset=0
     projectDown=30000
     projectUp=30000
     serializedChannel=ECC_GameTraceChannel6
     traceChannel=TraceTypeQuery7
     traceMode=GeometryByChannel
     traceShape=Sphere
    }
   }
  }
  tests [
   FortQueryTest_InsideAthenaSafeZone_1 {class=FortQueryTest_InsideAthenaSafeZone, properties {testPurpose=Filter}}
   EnvQueryTest_Overlap_2 {
    class=EnvQueryTest_Overlap
    properties {overlapData {extentX=700, extentY=700, extentZ=700, overlapChannel=ECC_GameTraceChannel1, shapeOffset=<0 0 0>}}
   }
   EnvQueryTest_Overlap_3 {
    class=EnvQueryTest_Overlap
    properties {overlapData {extentX=300, extentY=300, extentZ=300, overlapChannel=ECC_GameTraceChannel15, shapeOffset=<0 0 0>}}
   }
   FortQueryTest_Random_1 {class=FortQueryTest_Random, properties {floatValueMax=1, scoringEquation=SquareRoot, scoringFactor=1, testPurpose=Filter}}
  ]
 }]
